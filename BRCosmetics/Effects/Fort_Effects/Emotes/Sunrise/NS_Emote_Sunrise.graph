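# Niagara system for the Sunrise emote, kicking up sand and small rocks with three independent sprite emitters.
# Rocks mixes a trickle with a burst, falling under gravity and wind, with sprite size widened beyond an LOD distance.
# Sand_Body bursts large, faint wind-blown puffs, while Sand_Feet streams smaller puffs visible only in the first loop.
# Both sand emitters drive a dynamic material parameter over particle life; the exposed User.SkeletalMesh is unreferenced by modules shown.
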NiagaraSystem /BRCosmetics/Effects/Fort_Effects/Emotes/Sunrise/NS_Emote_Sunrise
dependencies {usedDataInterfaces [&NiagaraDataInterfaceCurve], usedDynamicInputs [], usedModules [&SystemState]}
emitters {
 Rocks {
  emitterUpdateScript [
   &EmitterState(Life Cycle Mode=::NewEnumerator1, Loop Behavior=::NewEnumerator1, Loop Delay=-0.1292, Loop Duration=2.1292, UseLoopDelay=true)
   &SpawnRate(SpawnRate=&RandomRangeFloat(Minimum=1, Maximum=2), Use Spawn Probability=false)
   &SpawnBurst_Instantaneous(Spawn Count=3, Spawn Time=0.9792, Use Spawn Probability=false)
  ]
  particleSpawnScript [
   /Niagara/Modules/Spawn/Initialization/InitializeParticle(
    Color=<rgba 0.2344 0.1266 0.0166 1>
    Module.Write Color=true
    Module.Write Lifetime=true
    Module.Write Mass=true
    Module.Write Material Random=true
    Module.Write Position=true
    Module.Write Scale=false
    Module.Write SpriteRotation=false
    Module.Write SpriteSize=true
    Sprite Size=&Vector2DFromFloat(Value=&MakeCustomFloatFromBool(
      Bool=HlslExpression() {hlslExpression="Engine.Owner.LODDistance > 310 "}
      True Float=&RandomRangeFloat(Minimum=1.25, Maximum=3)
      False Float=&RandomRangeFloat(Minimum=1, Maximum=2.2)
     ))
   )
   &SphereLocation(
    MaskSpawnGroup=false
    Offset=<9.3379 -3 -10>
    Override Seed=false
    Sphere Radius=5
    Surface Only=false
    UseNonUniformScale=false
    UseOffset=false
    UseSphereOrientation=false
   )
   &AddVelocityFromPoint(Module.Use Falloff Distance=false, Origin Offset=<5 5 0>, Velocity Strength=1)
  ]
  particleUpdateScript [
   &ParticleState()
   &GravityForce()
   &WindForce()
   &SolveForcesAndVelocity(Clamp Velocity=false, Limit Acceleration=false)
   &SubUVAnimation(End Frame=3)
   &Color(
    Color=<rgba 0.2344 0.1266 0.0166 1>
    Scale Alpha=&FloatFromCurve(FloatCurve=DataInterface() {propertyValues {Curve {keys [<at 0 1>,<at 1 0>]}}})
   )
   &ScaleSpriteSize(Uniform Curve Sprite Scale=DataInterface() {propertyValues {Curve {keys [
        {interpMode=RCIM_Cubic, tangentMode=RCTM_Break, time=0, value=1}
        {interpMode=RCIM_Cubic, tangentMode=RCTM_Break, time=0.6, value=1}
        {interpMode=RCIM_Cubic, tangentMode=RCTM_Break, time=1, value=0}
       ]}}})
  ]
  rendererClasses [&NiagaraSpriteRendererProperties]
  renderers [{rendererClass=&NiagaraSpriteRendererProperties}]
 }
 Sand_Body {
  emitterUpdateScript [
   &EmitterState(Life Cycle Mode=::NewEnumerator1, Loop Behavior=::NewEnumerator1, Loop Duration=1, UseLoopDelay=false)
   &SpawnBurst_Instantaneous(Spawn Count=8, Use Spawn Probability=false, UseLoopCountLimit=false)
  ]
  particleSpawnScript [
   /Niagara/Modules/Spawn/Initialization/V2/InitializeParticle(
    Color Mode=::NewEnumerator0
    Lifetime=&RandomRangeFloat(Minimum=2, Maximum=3)
    Sprite Rotation Angle=&RandomRangeFloat(Minimum=0, Maximum=360)
    Sprite Rotation Mode=::NewEnumerator0
    Sprite Size Mode=::NewEnumerator3
    Uniform Sprite Size=&RandomRangeFloat(Minimum=70, Maximum=90)
    UsePositionOffset=false
    Write Material Random=true
   )
   &SphereLocation(
    MaskSpawnGroup=false
    Non Uniform Scale=<2.2569 1 1>
    Offset=<0 -90 0>
    Override Seed=false
    Surface Expansion Mode=::NewEnumerator2
    Surface Only=true
    UseNonUniformScale=false
    UseOffset=false
    UseSphereOrientation=false
   )
   &AddVelocity(Velocity=<30 30 -60>)
  ]
  particleUpdateScript [
   &ParticleState()
   &WindForce(Wind Speed=<200 5 5>)
   &Drag(Use Linear Drag=true, Use Rotational Drag=false)
   &SpriteRotationRate(Rotation Rate=&RandomRangeFloat(Minimum=-30, Maximum=30))
   &DynamicMaterialParameters(
    Index 0 Param 1=&FloatFromCurve(FloatCurve=DataInterface() {propertyValues {Curve {keys [{interpMode=RCIM_Cubic, leaveTangent=-3.5204, tangentMode=RCTM_Break, time=0, value=1}, <at 1 -0.5>]}}})
    Module.Param0WriteEnabled=true
    Module.Param1WriteEnabled=true
    Module.Param2WriteEnabled=true
    Module.Param3WriteEnabled=true
   )
   &ScaleSpriteSize(Scale Factor=&Vector2DFromFloat(Value=&FloatFromCurve(
      FloatCurve=DataInterface() {propertyValues {Curve {keys [{interpMode=RCIM_Cubic, leaveTangent=1.5484, tangentMode=RCTM_Break, time=0, value=0.8}, <at 1 1.45>]}}}
      Scale Curve=&RandomRangeFloat(Minimum=0.8, Maximum=1.5)
     ))) {module=ScaleSpriteSize001}
   &Color(
    Color=<rgba 0.5672 0.3004 0.1659 1>
    Scale Alpha=&FloatFromCurve(FloatCurve=DataInterface() {propertyValues {Curve {keys [<at 0 0 RCIM_Cubic>,<at 0.5 1 RCIM_Cubic>,<at 1 0>]}}}, Scale Curve=0.025)
   )
   &SolveForcesAndVelocity(Clamp Velocity=false, Limit Acceleration=false)
   &SubUVAnimation()
  ]
  rendererClasses [&NiagaraSpriteRendererProperties]
  renderers [{rendererClass=&NiagaraSpriteRendererProperties}]
 }
 Sand_Feet {
  emitterUpdateScript [
   &EmitterState(Life Cycle Mode=::NewEnumerator1, Loop Behavior=::NewEnumerator1, Loop Duration=2, UseLoopDelay=true)
   &SpawnRate(SpawnRate=&RandomRangeFloat(Minimum=6, Maximum=10), Use Spawn Probability=false)
  ]
  particleSpawnScript [
   &InitializeParticle(
    Lifetime Max=1.5
    Lifetime Min=0.5
    Lifetime Mode=::NewEnumerator1
    Sprite Rotation Angle=&RandomRangeFloat(Minimum=0, Maximum=360)
    Sprite Rotation Mode=::NewEnumerator0
    Sprite Size Mode=::NewEnumerator1
    Uniform Sprite Size Max=30
    Uniform Sprite Size Min=15
    UsePositionOffset=false
    Write Material Random=true
   ) {module=InitializeParticle001}
   &SphereLocation(
    MaskSpawnGroup=false
    Override Seed=false
    Sphere Radius=5
    Surface Only=false
    UseNonUniformScale=false
    UseOffset=false
    UseSphereOrientation=false
   )
   &AddVelocity(Coordinate Space=::World, Velocity=&RandomRangeVector(Minimum=<0 0 0>, Maximum=<0 5 10>))
   &CameraOffset(Camera Offset Amount=8)
  ]
  particleUpdateScript [
   &ParticleState()
   &Drag(Drag=2, Use Linear Drag=true, Use Rotational Drag=false)
   &GravityForce(Gravity=&RandomRangeVector(Minimum=<0 0 -50>, Maximum=<1 1 -2>))
   &SpriteRotationRate(Rotation Rate=&RandomRangeFloat(Minimum=-30, Maximum=30))
   &WindForce(Wind Speed=&RandomRangeVector(Minimum=<20 -20 0>, Maximum=<60 60 1>))
   &AccelerationForce(Acceleration=<60 30 0>)
   &ScaleSpriteSize(Scale Factor=&Vector2DFromFloat(Value=&FloatFromCurve(
      FloatCurve=DataInterface() {propertyValues {Curve {keys [{interpMode=RCIM_Cubic, leaveTangent=1.5484, tangentMode=RCTM_Break, time=0, value=0.8}, <at 1 1.45>]}}}
      Scale Curve=&RandomRangeFloat(Minimum=0.8, Maximum=1.2354)
     )))
   &Color(
    Color=<rgba 0.4063 0.2152 0.1189 1>
    Scale Alpha=&FloatFromCurve(
     FloatCurve=DataInterface() {propertyValues {Curve {keys [
         {interpMode=RCIM_Cubic, leaveTangent=0.1136, tangentMode=RCTM_Break, time=0, value=0}
         {interpMode=RCIM_Cubic, leaveTangent=-1.8827, tangentMode=RCTM_Break, time=0.25, value=1}
         {arriveTangent=-0.027, interpMode=RCIM_Cubic, leaveTangent=-0.027, tangentMode=RCTM_User, time=1, value=0}
        ]}}}
     CurveIndex=$Emitter.NormalizedLoopAge
     Scale Curve=&MakeCustomFloatFromBool(
      Bool=HlslExpression() {hlslExpression="Emitter.LoopCount == 0"}
      True Float=&RandomRangeFloat(Minimum=0.5, Maximum=1)
      False Float=0
     )
    )
   ) {module=Color001}
   &DynamicMaterialParameters(
    Index 0 Param 1=&FloatFromCurve(FloatCurve=DataInterface() {propertyValues {Curve {keys [{interpMode=RCIM_Cubic, leaveTangent=-3.5204, tangentMode=RCTM_Break, time=0, value=1}, <at 1 -0.5>]}}})
    Module.Param0WriteEnabled=true
    Module.Param1WriteEnabled=true
    Module.Param2WriteEnabled=true
    Module.Param3WriteEnabled=true
   )
   &SolveForcesAndVelocity(Clamp Velocity=false, Limit Acceleration=false)
  ]
  rendererClasses [&NiagaraSpriteRendererProperties]
  renderers [{rendererClass=&NiagaraSpriteRendererProperties}]
 }
}
summary {
 systemName=NS_Emote_Sunrise
 emitters [
  {emitterName=Sand_Feet, rendererClasses [&NiagaraSpriteRendererProperties]}
  {emitterName=Rocks, rendererClasses [&NiagaraSpriteRendererProperties]}
  {emitterName=Sand_Body, rendererClasses [&NiagaraSpriteRendererProperties]}
 ]
 userVariables [User.SkeletalMesh {
   type=NiagaraDataInterfaceSkeletalMesh
   defaultValue {
    struct=NiagaraExt_VariableValue_DataInterface
    value {dataInterface=&self:NiagaraDataInterfaceSkeletalMesh_0}
   }
  }]
}
